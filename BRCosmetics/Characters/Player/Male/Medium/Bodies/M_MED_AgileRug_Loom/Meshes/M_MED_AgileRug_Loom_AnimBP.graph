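# Animation Blueprint for the M_MED_AgileRug_Loom player body cosmetic, applying secondary deformation to the character mesh.
# The AnimGraph copies the pose from the attached parent mesh, then runs it through the body's DeformRig Control Rig, passing in skydiving and front-end flags.
# Each update, the event graph sets the front-end flag true only when in the front end and not playing an emote.
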AnimBlueprint /BRCosmetics/Characters/Player/Male/Medium/Bodies/M_MED_AgileRug_Loom/Meshes/M_MED_AgileRug_Loom_AnimBP
target_skeleton=&/M_MED_AgileRug_Loom_Skeleton
graphs [
 AnimGraph {nodes [
   n1 = Control Rig!(Source=Copy Pose From Mesh() {bUseAttachedParent=true}, Alpha, skydiving:bool=Get bIsSkyDiving(), Front End:bool=Get Is FE() -> (Is FE:bool)) -> (Pose) {
    alphaBoolBlend {blendOption=Linear}
    controlRigAssetReference {blueprintRigClass=&/M_MED_AgileRug_Loom_DeformRig.M_MED_AgileRug_Loom_DeformRig_C}
   }
   Output Pose(Result=@n1) {layerGroup=None}
  ]}
 EventGraph {nodes [
   Event BlueprintUpdateAnimation()
   n1 = Get bIsPlayingEmote()
   > Set Is FE(Is FE:bool=AND Boolean(A=Get bIsInFrontEnd(), B=NOT Boolean(A=@n1))) -> (Output_Get:bool)
  ]}
]
variables [Is FE {type=boolean}]
